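# Translucent, unlit, two-sided Niagara sprite material for the Relic Finder hologram, built on a sprite dust icon texture.
# The hue-shifted, dimmed texture is overlay-blended with purple plus animated scanlines, then passed through the shared FX emissive adjustment function.
# Texture alpha times particle alpha forms the opacity term.
# A time-driven hologram wiggle function plus a sine bob along Z forms the position offset.
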Material /JobBoard/FX/M_RelicFinder_Hologram
graph {
 properties {
  bAutomaticallySetUsageInEditor=false
  bUsedWithNiagaraSprites=true
  blendMode=BLEND_Translucent
  shadingModel=MSM_Unlit
  twoSided=true
  usedShadingModels=Unlit
 }
 nodes [
  n1 = Time()
  Add(
   A=MaterialFunctionCall!(
    WPO Intensity A=Constant3Vector() {constant=<rgba 0.25 0.25 0.25 0>}
    WPO Intensity B=Constant3Vector() {constant=<rgba 1 1 1 0>}
    Time=@n1
    Bias
   ) {materialFunction=/Game/Effects/Fort_Effects/Effects/Atlas/Materials/MF_Atlas_HologramWiggle_01}.Result
   B=Multiply(A=Sine(None=@n1) {period=2}, B=Constant3Vector() {constant=<rgba 0 0 5 0>})
  )
  n2 = TextureSample!(UVs, Tex, Level, Apply View MipBias) {
   automaticViewMipBias=false
   mipValueMode=TMVM_MipLevel
   samplerSource=SSM_Clamp_WorldGroupSettings
   texture=&/T_UI_SpriteExtraction_SpriteDust_Color_128
  }
  n3 = MaterialFunctionCall(Hue Shift Percentage=Constant() {r=0.05}, Texture=@n2.RGB) {materialFunction=&HueShift}
  n4 = Constant3Vector() {constant=<rgba 0.2871 0 1 1>}
  n5 = MaterialFunctionCall!(
   Number of Scanlines=Constant() {r=3}
   Scanline Speed=Constant() {r=1}
   "Scanlines Thickness (Bias)"=Constant() {r=3}
   Scanlines Dark Value=Constant()
   Scanlines Light Value=Constant() {r=0.1}
   Block Opacity=Constant() {r=1}
   UVs=TextureCoordinate() {uTiling=0, vTiling=5}
   Time=@n1
   Block Repeat
  ) {materialFunction=/Game/Effects/Fort_Effects/Effects/Atlas/Materials/MF_Atlas_Scanlines_01}
  n6 = Add(A=@n5.Result, B=@n5.Result)
  MaterialFunctionCall(Emissive=MaterialFunctionCall(Base=Multiply(B=@n3.Result) {constA=0.75}, Blend=Add(A=@n4, B=@n6)) {materialFunction=&Blend_Overlay}.Result) {materialFunction=&MF_FX_EmissiveAdjustments}
  Multiply(A=Multiply(A=@n2.A), B=ParticleColor().A)
 ]
}
