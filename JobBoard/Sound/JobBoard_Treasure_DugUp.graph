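# Sound cue for JobBoard treasure being dug up, built from gold-vein pickaxe destroy waves.
# A Quality Level node sends the first two levels to a random pick of eight variants and the last two to a pick of the first four.
# A Modulator at the root adds slight random variation.
# Shared gold-vein attenuation, the weapon-impact sound class and a small concurrency cap govern playback.
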SoundCue /JobBoard/Sound/JobBoard_Treasure_DugUp
root=Modulator_1
nodes [
 {id=Modulator_1, children [QualityLevel_0], properties {volumeMax=1, volumeMin=0.9}}
 {id=QualityLevel_0, children [Random_2,Random_2,Random_3,Random_3]}
 {
  id=Random_2
  children [WavePlayer_18, WavePlayer_19, WavePlayer_20, WavePlayer_21, WavePlayer_22, WavePlayer_23, WavePlayer_24, WavePlayer_25]
  properties {weights [1,1,1,1,1,1,1,1]}
 }
 {id=Random_3, children [WavePlayer_26, WavePlayer_27, WavePlayer_28, WavePlayer_29], properties {weights [1,1,1,1]}}
 {id=WavePlayer_18, properties {soundWaveAssetPtr=/JobBoard/Sound/Waves/GoldVein_Destroy_Pickaxe_01}}
 {id=WavePlayer_19, properties {soundWaveAssetPtr=/JobBoard/Sound/Waves/GoldVein_Destroy_Pickaxe_02}}
 {id=WavePlayer_20, properties {soundWaveAssetPtr=/JobBoard/Sound/Waves/GoldVein_Destroy_Pickaxe_03}}
 {id=WavePlayer_21, properties {soundWaveAssetPtr=/JobBoard/Sound/Waves/GoldVein_Destroy_Pickaxe_04}}
 {id=WavePlayer_22, properties {soundWaveAssetPtr=/JobBoard/Sound/Waves/GoldVein_Destroy_Pickaxe_05}}
 {id=WavePlayer_23, properties {soundWaveAssetPtr=/JobBoard/Sound/Waves/GoldVein_Destroy_Pickaxe_06}}
 {id=WavePlayer_24, properties {soundWaveAssetPtr=/JobBoard/Sound/Waves/GoldVein_Destroy_Pickaxe_07}}
 {id=WavePlayer_25, properties {soundWaveAssetPtr=/JobBoard/Sound/Waves/GoldVein_Destroy_Pickaxe_08}}
 {id=WavePlayer_26, properties {soundWaveAssetPtr=&GoldVein_Destroy_Pickaxe_01}}
 {id=WavePlayer_27, properties {soundWaveAssetPtr=&GoldVein_Destroy_Pickaxe_02}}
 {id=WavePlayer_28, properties {soundWaveAssetPtr=&GoldVein_Destroy_Pickaxe_03}}
 {id=WavePlayer_29, properties {soundWaveAssetPtr=&GoldVein_Destroy_Pickaxe_04}}
]
properties {
 attenuationSettings=/JobBoard/Sound/Settings/ATT_GoldVein_Destroy
 bOverrideConcurrency=true
 maxDistance=2512
 soundClassObject=/Game/Sounds/FortSoundClassesAndModes/Fort_Events_Weapons_Impact
 volumeMultiplier=1
 concurrencyOverrides {maxCount=3}
}
