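# Ambient loop Sound Cue for the JobBoard, mixing an idle bed with intermittent pings under the loot ambient sound class and concurrency rules.
# The bed randomly picks one of two looping idle waves, fades it in through an enveloper and uses chest attenuation.
# The quieter ping layer repeatedly waits a random delay, then plays one of four pings with modulated pitch and volume, using its own logarithmic attenuation override.
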SoundCue /JobBoard/Sounds/JobBoard_Ambient_Loop_Cue
root=Mixer_0
nodes [
 {id=Attenuation_0, children [Enveloper_0], properties {attenuationSettings=/Game/Sounds/Attenuation/Fort_Chests_Att}}
 {
  id=Attenuation_1
  children [Looping_0]
  properties {
   attenuationSettings=&/Fort_JobBoard_Ambient_Loop_Mono_Default_Att
   attenuationOverrides {distanceAlgorithm=Logarithmic, falloffDistance=3000}
  }
 }
 {id=Delay_0, children [Modulator_0], properties {delayMax=8, delayMin=2}}
 {
  id=Enveloper_0
  children [Random_0]
  properties {pitchCurve {editorCurveData {keys [{time=0}]}}, volumeCurve {editorCurveData {keys [<at 0 0>,<at 1 1>]}}}
 }
 {id=Looping_0, children [Delay_0]}
 {id=Mixer_0, children [Attenuation_0, Attenuation_1], properties {inputVolume [1,0.75]}}
 {id=Modulator_0, children [Random_2], properties {pitchMin=0.85, volumeMax=1, volumeMin=0.85}}
 {id=Random_0, children [WavePlayer_1,WavePlayer_2], properties {weights [1,1]}}
 {id=Random_2, children [WavePlayer_0,WavePlayer_3,WavePlayer_4,WavePlayer_5], properties {weights [1,1,1,1]}}
 {id=WavePlayer_0, properties {soundWaveAssetPtr=&/JobBoard_Ping_01}}
 {id=WavePlayer_1, properties {bLooping=true, soundWaveAssetPtr=&/JobBoard_Idle_01}}
 {id=WavePlayer_2, properties {bLooping=true, soundWaveAssetPtr=&/JobBoard_Idle_02}}
 {id=WavePlayer_3, properties {soundWaveAssetPtr=&/JobBoard_Ping_02}}
 {id=WavePlayer_4, properties {soundWaveAssetPtr=&/JobBoard_Ping_03}}
 {id=WavePlayer_5, properties {soundWaveAssetPtr=&/JobBoard_Ping_04}}
]
properties {
 maxDistance=3500
 soundClassObject=/Game/Sounds/FortSoundClassesAndModes/Fort_Loot_Ambient_Loop
 volumeMultiplier=1
 concurrencyOverrides {maxCount=2}
 concurrencySet [/Game/Sounds/Concurrency_Rules/Loot_Ambient_Conc]
}
