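# Sound cue for the job board's deactivation, playing a power-down sound.
# A root Random node picks one of three PowerDown wave variations with equal weights.
# It uses the shared item-interact attenuation asset and the loot-reward sound class, with a boosted volume.
# Overridden concurrency allows one instance per owner, stopping the quietest on conflict.
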SoundCue /JobBoard/Sounds/JobBoard_Deactivate_Cue
root=Random_1
nodes [
 {id=Random_1, children [WavePlayer_0,WavePlayer_5,WavePlayer_6], properties {weights [1,1,1]}}
 {id=WavePlayer_0, properties {soundWaveAssetPtr=&/JobBoard_PowerDown_01}}
 {id=WavePlayer_5, properties {soundWaveAssetPtr=&/JobBoard_PowerDown_02}}
 {id=WavePlayer_6, properties {soundWaveAssetPtr=&/JobBoard_PowerDown_03}}
]
properties {
 attenuationSettings=/Game/Sounds/Attenuation/Fort_Item_Interact_Att
 bOverrideConcurrency=true
 maxDistance=1850
 soundClassObject=&Fort_Loot_Reward_PSM
 volumeMultiplier=1.5
 attenuationOverrides {
  bEnableFocusInterpolation=true
  bEnableListenerFocus=true
  distanceAlgorithm=Logarithmic
  falloffDistance=1000
  focusAttackInterpSpeed=20
  focusAzimuth=20
  focusReleaseInterpSpeed=10
  focusVolumeAttenuation=1.6
  lPFFrequencyAtMax=10000
  lPFRadiusMax=1000
  lPFRadiusMin=500
  nonFocusAzimuth=40
  nonFocusVolumeAttenuation=0.6
  nonSpatializedRadiusStart=200
  occlusionInterpolationTime=0.3
  occlusionLowPassFilterFrequency=3000
  occlusionVolumeAttenuation=0.75
  attenuationShapeExtents {x=0}
  customAttenuationCurve {editorCurveData {keys [<at 0 1 RCIM_Cubic>, {arriveTangent=-1, interpMode=RCIM_Cubic, leaveTangent=-1, time=0.5, value=0.1}, <at 1 0 RCIM_Cubic>]}}
 }
 concurrencyOverrides {bLimitToOwner=true, maxCount=1, resolutionRule=StopQuietest}
}
