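# Sound cue for cancelling or closing the Job Board's shadow briefing.
# A Random node picks one of three JobBoard_Close wave players with equal weight, giving variation on each play.
# It uses the shared item-interact attenuation and plays through the loot reward sound class.
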SoundCue /JobBoard/Sounds/JobBoard_ShadowBriefing_Cancel_Cue
root=Random_0
nodes [
 {id=Random_0, children [WavePlayer_0,WavePlayer_1,WavePlayer_2], properties {weights [1,1,1]}}
 {id=WavePlayer_0, properties {soundWaveAssetPtr=&/JobBoard_Close_03}}
 {id=WavePlayer_1, properties {soundWaveAssetPtr=&/JobBoard_Close_02}}
 {id=WavePlayer_2, properties {soundWaveAssetPtr=&/JobBoard_Close_01}}
]
properties {
 attenuationSettings=/Game/Sounds/Attenuation/Fort_Item_Interact_Att
 maxDistance=1850
 soundClassObject=&Fort_Loot_Reward_PSM
}
