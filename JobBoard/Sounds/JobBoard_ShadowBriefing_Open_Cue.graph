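# Sound cue for the Job Board Shadow Briefing opening, pairing an open one-shot with a looping active ambience.
# A mixer plays one of three equally weighted open sounds alongside the quieter active loop.
# The loop gets random pitch variation, starts after a short delay and fades in through a volume envelope.
# It uses item-interact attenuation and a loot-reward sound class.
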SoundCue /JobBoard/Sounds/JobBoard_ShadowBriefing_Open_Cue
root=Mixer_0
nodes [
 {id=Delay_0, children [Enveloper_0], properties {delayMax=0.1, delayMin=0.1}}
 {
  id=Enveloper_0
  children [Modulator_0]
  properties {
   pitchCurve {editorCurveData {keys [{time=0}]}}
   volumeCurve {editorCurveData {keys [{arriveTangent=-0.1692, interpMode=RCIM_Cubic, leaveTangent=-0.1692, tangentMode=RCTM_User, time=0, value=0.01}, <at 0.5 1>]}}
  }
 }
 {id=Mixer_0, children [Random_1,Delay_0], properties {inputVolume [1,0.65]}}
 {id=Modulator_0, children [WavePlayer_0], properties {pitchMax=1.1, pitchMin=0.9}}
 {id=Random_1, children [WavePlayer_3,WavePlayer_4,WavePlayer_5], properties {weights [1,1,1]}}
 {id=WavePlayer_0, properties {bLooping=true, soundWaveAssetPtr=&/JobBoard_Active_Loop}}
 {id=WavePlayer_3, properties {soundWaveAssetPtr=&/JobBoard_Open_03}}
 {id=WavePlayer_4, properties {soundWaveAssetPtr=&/JobBoard_Open_02}}
 {id=WavePlayer_5, properties {soundWaveAssetPtr=&/JobBoard_Open_01}}
]
properties {
 attenuationSettings=/Game/Sounds/Attenuation/Fort_Item_Interact_Att
 maxDistance=1850
 soundClassObject=&Fort_Loot_Reward_PSM
}
