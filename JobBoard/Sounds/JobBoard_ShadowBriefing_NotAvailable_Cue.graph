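# JobBoard_ShadowBriefing_NotAvailable_Cue is a SoundCue in the JobBoard content that, by its name, signals a Shadow Briefing option is not available.
# A single Wave Player feeds the output directly, reusing the NPC conversation radial dial error sound with no randomization or modulation.
# It uses the non-critical in-game UI sound class.
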SoundCue /JobBoard/Sounds/JobBoard_ShadowBriefing_NotAvailable_Cue
root=WavePlayer_1
nodes [{id=WavePlayer_1, properties {soundWaveAssetPtr=/Game/Athena/Sounds/NPCs/Conversation/NPC_Radial_Dial_Error_01}}]
properties {maxDistance=2097152, soundClassObject=&Fort_UI_InGame_NonCritical}
